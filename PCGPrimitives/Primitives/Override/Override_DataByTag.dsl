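; Override_DataByTag is a reusable PCG graph that replaces data in its primary input with data from isolated actors, based on the IsolatedDataTags parameter.
; Get Actor Data selects only isolated actors, reading their PCG component output or parsing their components.
; That data has its PCG component and type tags stripped and gains an "Overriden" tag.
; Replace Data By Tag then uses it as the replacement and outputs the result.
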(graph Override_DataByTag
  (params (IsolatedDataTags :type String))
  (inputs (PrimaryInput :type Any))
  (bind GetActorData_1 (Get%20Actor%20Data :ActorSelector (struct :IncludeIsolatedActors Only) :bIgnorePCGGeneratedComponents false :Mode GetDataFromPCGComponentOrParseComponents))
  (bind DeleteTags_0 (Delete%20Tags :SelectedTags "PCG Component, Type:Point, Type:Spline, Type:ClosedSpline, Type:Polygon2D, Type:Surface, Type:Texture2D, Type:Volume"))
  (bind AddTags_0 (Add%20Tags :TagsToAdd "Overriden"))
  (bind ReplaceDataByTag_1 (Replace%20Data%20By%20Tag :Tags-param IsolatedDataTags))
  (connect DeleteTags_0 Out AddTags_0 In)
  (connect GetActorData_1 Out DeleteTags_0 In)
  (connect DefaultInputNode PrimaryInput ReplaceDataByTag_1 In)
  (connect AddTags_0 Out ReplaceDataByTag_1 Replacement)
  (output Out ReplaceDataByTag_1))
